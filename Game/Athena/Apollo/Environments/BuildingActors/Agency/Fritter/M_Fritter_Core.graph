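# Unlit translucent particle material for the Fritter energy core effect, with a scalar blend between mesh and sprite looks.
# Distorted panning noise drives a two-color glow, to which mesh mode adds a Fresnel-indexed curve-atlas color, all tinted by particle color.
# Opacity uses a noise-driven sphere mask for sprites or view-angle edge fade with optional curve-based vertical fade for meshes.
# Panning noise along vertex normals forms a maskable vertex offset, zero by default.
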Material /Game/Athena/Apollo/Environments/BuildingActors/Agency/Fritter/M_Fritter_Core
graph {
 properties {
  bUsedWithMeshParticles=true
  bUsedWithNiagaraMeshParticles=true
  bUsedWithParticleSprites=true
  blendMode=BLEND_Translucent
  pixelDepthOffsetMode=PDOM_Legacy
  shadingModel=MSM_Unlit
  usedShadingModels=Unlit
 }
 nodes [
  n1 = VectorParameter() {parameterName=Sprite_ColorHigh, defaultValue=<rgba 1 0 0 1>}
  n2 = VectorParameter() {parameterName=Tiling, defaultValue=<rgba 1 2 1 1>}
  n3 = Multiply(A=TextureCoordinate() {vTiling=2}, B=AppendVector(A=@n2.R, B=@n2.R))
  n4 = AppendVector(A=@n2.B, B=@n2.B)
  n5 = TextureCoordinate()
  n6 = TextureSampleParameter2D(UVs=Panner(Coordinate=Multiply(A=@n4, B=@n5)) {speedX=-0.02, speedY=0.1}) {parameterName=DistortTexture, samplerType=SAMPLERTYPE_LinearColor, texture=&T_NoisePack_LinearColor}
  n7 = Divide(A=ComponentMask(None=@n6.RGB) {g=true, r=true}, B=ScalarParameter() {defaultValue=2, parameterName=DistortAmount})
  n8 = TextureSampleParameter2D(UVs=Panner(Coordinate=Add(A=@n3, B=@n7)) {speedX=0.2, speedY=-0.25}) {parameterName=NoiseA, texture=/Game/Weapons/WeaponSkins/Wraps/Textures/Mash/Mash_M}
  n9 = Multiply(A=@n1.RGB, B=Desaturation(None=@n8.RGB, Fraction=ScalarParameter() {parameterName=Desaturation}))
  n10 = CameraVectorWS()
  n11 = VertexNormalWS()
  n12 = Power(Base=Abs(None=DotProduct(A=@n10, B=@n11))) {constExponent=4}
  n13 = ScalarParameter() {defaultValue=1, parameterName="UseForMesh(1)OrSprite(0)"}
  n14 = LinearInterpolate(
   A=VectorParameter() {parameterName=Sprite_ColorLow, defaultValue=<rgba 5 0.925 0.165 1>}.RGB
   B=@n1.RGB
   Alpha=Multiply(A=Multiply(A=@n8.R, B=@n8.R), B=LinearInterpolate(B=@n12, Alpha=@n13) {constA=1})
  )
  n15 = CurveAtlasRowParameter(CurveTime=Fresnel() {baseReflectFraction=0, exponent=1}) {
   atlas=&/FritterCurveAtlas
   curve=&/Curve_Fritter_Color_EnergyCore
   defaultValue=2
   parameterName=MeshColorCurve
  }
  n16 = Add(
   A=Multiply(A=Multiply(A=@n14) {constB=30}, B=@n8.R)
   B=Multiply(A=@n15, B=ScalarParameter() {defaultValue=10, parameterName=MeshEmissiveBrightness})
  )
  n17 = ScalarParameter() {defaultValue=1, parameterName="UseForMesh(1)OrSprite(0)"}
  n18 = ParticleColor()
  MaterialFunctionCall(Emissive=Multiply(A=LinearInterpolate(A=@n9, B=@n16, Alpha=@n17), B=@n18.RGB)) {materialFunction=&MF_FX_EmissiveAdjustments}
  n19 = SphereMask(
   A=TextureCoordinate()
   B=Constant() {r=0.5}
   Radius=LinearInterpolate(A=ScalarParameter() {defaultValue=0.2, parameterName=SphereMaskDistortion}, Alpha=@n6.B) {constA=0.2, constB=0.5}
  ) {attenuationRadius=0.5, hardnessPercent=0}
  n20 = ScalarParameter() {defaultValue=3, parameterName=EdgeFadeExp}
  n21 = Saturate(None=Multiply(
    A=MaterialFunctionCall("FadeSensitivity (optional)"=@n20) {materialFunction=&FadeBasedOnViewAngle}.Result
    B=ScalarParameter() {defaultValue=3, parameterName=EdgeFadeHardness}
   ) {constB=3})
  n22 = MaterialFunctionCall() {materialFunction=&LinearGradient}
  n23 = CurveAtlasRowParameter(CurveTime=@n22.VGradient) {
   atlas=&FritterCurveAtlas
   curve=&Curve_Fritter_Color_EnergyCore
   defaultValue=2
   parameterName=MeshOpacityCurve
  }
  n24 = Multiply(A=@n21, B=@n23.A)
  n25 = ScalarParameter() {parameterName=UseFadeTopAndBottom}
  Multiply(
   A=LinearInterpolate(
    A=Multiply(A=@n19, B=@n19)
    B=LinearInterpolate(A=@n21, B=@n24, Alpha=@n25)
    Alpha=ScalarParameter() {defaultValue=1, parameterName="UseForMesh(1)OrSprite(0)"}
   ) {constA=1}
   B=@n18.A
  )
  n26 = ScalarParameter() {defaultValue=1, parameterName=WPO_Tiling}
  Multiply(A=TextureCoordinate(), B=@n26)
  n27 = MaterialFunctionCall() {materialFunction=&LinearGradient}
  n28 = Sine(None=Multiply(A=@n27.VGradient) {constB=0.5})
  n29 = ScalarParameter() {defaultValue=1, parameterName=WPO_Mask}
  n30 = Multiply(A=@n27.VGradient, B=@n26)
  n31 = TextureSample(UVs=Panner(Coordinate=AppendVector(A=@n27.UGradient, B=@n30)) {speedY=-0.3}) {samplerType=SAMPLERTYPE_LinearColor, texture=&T_NoisePack_LinearColor}
  n32 = VertexNormalWS()
  Multiply(A=Multiply(A=LinearInterpolate(B=@n28, Alpha=@n29) {constA=1}, B=Multiply(A=@n31.RGB, B=@n32)), B=ScalarParameter() {parameterName=WPO})
  Sine(None=Multiply(A=@n22.VGradient) {constB=0.5})
 ]
}
