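# Translucent, two-sided unlit material for Fritter's glowing energy arms, built for spline meshes and mesh particles.
# World position offset combines panned dual-noise wobble, a pulsing flow along normals, SplineThicken width and a sine bend, with MPC_Fritter_EnergyWave feeding bend and timing.
# Noise-distorted sine bands and gradient masks shape the HDR color and opacity, with optional flicker.
# Dynamic parameters drive beam growth, wobble intensity, width and time offset per particle.
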Material /Game/Athena/Apollo/Environments/BuildingActors/Agency/Fritter/M_Fritter_EnergyArms
graph {
 properties {
  bUsedWithMeshParticles=true
  bUsedWithNiagaraMeshParticles=true
  bUsedWithSplineMeshes=true
  blendMode=BLEND_Translucent
  pixelDepthOffsetMode=PDOM_Legacy
  shadingModel=MSM_Unlit
  twoSided=true
  usedShadingModels=Unlit
 }
 nodes [
  n1 = Multiply(
   A=LinearInterpolate(
    A=Time()
    B=CollectionParameter() {collection=&/MPC_Fritter_EnergyWave, parameterName=Loudness}
    Alpha=ScalarParameter() {parameterName=UseMPC_Speed}
   )
   B=ScalarParameter() {defaultValue=1, parameterName=WPO_Speed}
  )
  Add(A=@n1, B=ComponentMask(None=ActorPositionWS()) {r=true})
  n2 = Multiply(
   A=Power(
    Base=TextureCoordinate() {coordinateIndex=1, uTiling=0, vTiling=1.5}
    Exp=ScalarParameter() {defaultValue=1, parameterName=EnergyFlowStretchExp}
   )
   B=ScalarParameter() {defaultValue=1.5, parameterName=EnergyFlowTiling}
  )
  n3 = ScalarParameter() {parameterName=TimeOffset}
  n4 = DynamicParameter() {paramNames [BeamGrowth, WPO_Intensity, SplineWidth, TimeOffset]}
  n5 = Add(A=Multiply(A=Add(A=@n3, B=@n4.TimeOffset), B=ScalarParameter() {defaultValue=1, parameterName=EnergyFlowDynamicCtrl}) {constB=-1}, B=Time())
  n6 = AppendVector(A=Constant(), B=ScalarParameter() {defaultValue=1, parameterName=EnergyFlowSpeed})
  n7 = LinearInterpolate(Alpha=TextureSampleParameter2D(UVs=Panner(Coordinate=@n2, Time=@n5, Speed=@n6)) {parameterName=PulsingNoise, texture=/Game/Global/Textures/Default/Tileable_Patterns/T_WavePattern_01}.R) {constB=4}
  n8 = VertexNormalWS()
  n9 = CurveAtlasRowParameter(CurveTime=MaterialFunctionCall() {materialFunction=&LinearGradient}.VGradient) {
   atlas=&/FritterCurveAtlas
   curve=&/Curve_Fritter_FX_Mask
   parameterName=WPO_MaskCurve
  }
  n10 = TextureCoordinate()
  n11 = Constant() {r=0.5}
  n12 = SphereMask(
   A=Power(Base=ComponentMask(None=@n10) {g=true}, Exp=ScalarParameter() {defaultValue=1.7, parameterName=WPOMask_LinearGradient_CenterOffset}) {constExponent=1.7}
   B=@n11
   Radius=@n11
  ) {hardnessPercent=0}
  n13 = LinearInterpolate(
   A=@n9.R
   B=Saturate(None=Multiply(A=@n12) {constB=2})
   Alpha=ScalarParameter() {defaultValue=1, desc=Switch, parameterName="WPOMask_LinearGradient(1)Curve(0)"}
  ) {constA=1}
  n14 = Multiply(A=Multiply(A=Multiply(A=@n7, B=@n8), B=ScalarParameter() {parameterName=WPO_Intensity_Pulsing}), B=@n13)
  n15 = MaterialFunctionCall(Float2=TextureCoordinate()) {materialFunction=&BreakOutFloat2Components}
  n16 = VectorParameter() {parameterName=WPO_Tile, defaultValue=<rgba 1 1 1 0.5>}
  n17 = Multiply(A=@n15.R, B=@n16.R)
  n18 = Multiply(A=@n15.G, B=@n16.G)
  n19 = Add(
   A=Add(
    A=ScalarParameter() {parameterName=TimeOffset}
    B=DynamicParameter() {paramNames [BeamGrowth, WPO_Intensity, SplineWidth, TimeOffset]}.TimeOffset
   )
   B=@n1
  )
  n20 = TextureSampleParameter2D(UVs=Panner(Coordinate=AppendVector(A=@n17, B=@n18), Time=@n19) {speedX=0.03, speedY=-0.1}) {parameterName=WPO_Tex_A, samplerType=SAMPLERTYPE_LinearColor, texture=&T_NoisePack_LinearColor}
  n21 = Multiply(A=@n15.R, B=@n16.B)
  n22 = Multiply(A=@n15.G, B=@n16.A)
  n23 = TextureSampleParameter2D(UVs=Panner(Coordinate=AppendVector(A=@n21, B=@n22), Time=@n19) {speedX=-0.02, speedY=-0.05}) {parameterName=WPO_Tex_B, samplerType=SAMPLERTYPE_LinearColor, texture=&T_NoisePack_LinearColor}
  n24 = MaterialFunctionCall(Float3=Add(A=@n20.RGB, B=@n23.RGB) {constB=0}) {materialFunction=&BreakOutFloat3Components}
  n25 = MaterialFunctionCall(R=LinearInterpolate(Alpha=@n24.R) {constA=-1}, G=LinearInterpolate(Alpha=@n24.G) {constA=-1}, B=LinearInterpolate(Alpha=@n24.B) {constA=-1}) {materialFunction=&AppendMany}
  n26 = Transform(None=@n25.RGB)
  n27 = ScalarParameter() {parameterName=WPO_TransformTangetToWorld}
  n28 = ScalarParameter() {defaultValue=20, parameterName=WPO_Intensity}
  n29 = DynamicParameter() {paramNames [BeamGrowth, WPO_Intensity, SplineWidth, TimeOffset]}
  n30 = Multiply(A=@n13, B=Multiply(A=LinearInterpolate(A=@n25.RGB, B=@n26, Alpha=@n27), B=Multiply(A=@n28, B=@n29.WPO_Intensity)) {constB=20})
  n31 = ScalarParameter() {defaultValue=15, parameterName=EnergyArmWidth}
  n32 = Divide(A=@n31)
  n33 = Constant() {r=1}
  n34 = MaterialFunctionCall() {materialFunction=&LinearGradient}
  n35 = OneMinus(None=@n34.VGradient)
  n36 = ScalarParameter() {parameterName="UseBeamGrowth(Invert)"}
  n37 = DynamicParameter() {paramNames [BeamGrowth, WPO_Intensity, SplineWidth, TimeOffset]}
  n38 = Saturate(None=Subtract(A=LinearInterpolate(A=@n34.VGradient, B=@n35, Alpha=@n36), B=LinearInterpolate(Alpha=@n37.BeamGrowth) {constA=2, constB=-2}))
  n39 = ScalarParameter() {desc=Switch, parameterName=UseBeamGrowth}
  n40 = Multiply(A=Multiply(A=LinearInterpolate(A=@n32, B=@n31, Alpha=@n23.G), B=LinearInterpolate(A=@n33, B=@n38, Alpha=@n39)), B=@n29.SplineWidth)
  n41 = ChannelMaskParameter(None=Add(
    A=TextureCoordinate()
    B=CollectionParameter() {collection=&MPC_Fritter_EnergyWave, parameterName=TimeOffset}
   )) {maskChannel=Green, parameterName=ArcMmaskChannel, defaultValue=<rgba 0 1 0 0>}
  n42 = Multiply(
   A=ScalarParameter() {defaultValue=1, parameterName=WPO_BendTiling}
   B=CollectionParameter() {collection=&MPC_Fritter_EnergyWave, parameterName=WPO_BendTiling}
  )
  n43 = Sine(None=Multiply(A=Multiply(A=@n41.R, B=@n42)) {constB=0.5})
  n44 = CollectionParameter() {collection=&MPC_Fritter_EnergyWave, parameterName=WPO_BendIntensity}
  n45 = Multiply(
   A=Constant3Vector() {constant=<rgba 0 0 1 0>}
   B=Multiply(A=ScalarParameter() {parameterName=WPO_BendIntensity}, B=Multiply(A=@n44) {constB=100})
  )
  n46 = MaterialFunctionCall() {materialFunction=&LinearGradient}
  n47 = OneMinus(None=@n46.VGradient)
  n48 = Saturate(None=Multiply(A=@n46.VGradient) {constB=3.3})
  Add(
   A=Add(
    A=Add(A=@n14, B=@n30)
    B=MaterialFunctionCall(WidthBase=@n40, WidthTip=@n40) {materialFunction=&SplineThicken}.WorldPositionOffset
   )
   B=Multiply(A=Multiply(A=@n43, B=@n45) {constA=1}, B=Multiply(A=@n47, B=@n48))
  )
  LinearInterpolate(A=@n25.RGB, B=Multiply(A=VertexColor().R, B=@n25.RGB), Alpha=ScalarParameter() {parameterName=WPOMask_VertexColor})
  MaterialFunctionCall() {materialFunction=&DebugTimeSine}
  n49 = TextureCoordinate()
  n50 = TextureCoordinate() {uTiling=0.5, vTiling=5}
  n51 = Panner(Coordinate=@n50) {speedY=-0.1}
  n52 = Add(
   A=Multiply(
    A=MaterialFunctionCall(Float2=@n49) {materialFunction=&BreakOutFloat2Components}.R
    B=ScalarParameter() {defaultValue=15, parameterName=WaveTiling}
   )
   B=Multiply(A=TextureSample(UVs=@n51) {samplerType=SAMPLERTYPE_LinearColor, texture=&T_NoisePack_LinearColor}.R) {constA=3, constB=2}
  )
  n53 = Multiply(A=Time()) {constB=-2}
  n54 = Saturate(None=Sine(None=Add(A=@n52, B=@n53)))
  n55 = MaterialFunctionCall() {materialFunction=&LinearGradient}
  n56 = Sine(None=Multiply(A=@n55.UGradient) {constB=0.5})
  n57 = Saturate(None=@n56)
  n58 = LinearInterpolate(
   A=VectorParameter() {parameterName=ColorLow, defaultValue=<rgba 1 0.2066 0.0969 1>}.RGB
   B=VectorParameter() {parameterName=ColorHigh, defaultValue=<rgba 1 0.4803 0.1953 1>}.RGB
   Alpha=Multiply(A=@n54, B=Power(Base=@n57) {constExponent=4})
  )
  n59 = Multiply(A=@n58, B=@n54)
  n60 = ParticleColor()
  n61 = Saturate(None=Multiply(A=Add(A=@n58, B=@n59), B=@n60.RGB))
  n62 = Multiply(A=@n60.RGB, B=ScalarParameter() {defaultValue=15, parameterName=Brightness})
  n63 = MaterialFunctionCall() {materialFunction=&LinearGradient}
  n64 = CurveAtlasRowParameter(CurveTime=@n63.UGradient) {
   atlas=&FritterCurveAtlas
   curve=&Curve_Fritter_FX_Mask
   parameterName=Opacity_MaskCurve_U
  }
  n65 = CurveAtlasRowParameter(CurveTime=@n63.VGradient) {
   atlas=&FritterCurveAtlas
   curve=&Curve_Fritter_FX_Mask
   parameterName=Opacity_MaskCurve_V
  }
  n66 = LinearInterpolate(
   A=Multiply(A=@n64.B, B=@n65.G)
   B=Saturate(None=Multiply(A=@n55.VGradient) {constB=10})
   Alpha=ScalarParameter() {defaultValue=1, desc=Switch, parameterName="OpacityMask_LinearGradient(1)Curve(0)"}
  ) {constA=1}
  n67 = Subtract(A=Saturate(None=Multiply(A=@n56) {constB=5}), B=@n56)
  n68 = Multiply(A=@n66, B=LinearInterpolate(A=ScalarParameter() {defaultValue=0.1, parameterName=Opacity_Center}, Alpha=Saturate(None=@n67)) {constA=0.1})
  n69 = LinearInterpolate(B=@n7, Alpha=ScalarParameter() {defaultValue=1, parameterName=EnergyFlow}) {constA=1}
  MaterialFunctionCall(Emissive=LinearInterpolate(
    A=VectorParameter() {parameterName=Color_Dark, defaultValue=<rgba 0.0938 0.0067 0.0049 1>}.RGB
    B=MaterialFunctionCall(LDR Input=@n61, HDR Tint=@n62) {materialFunction=&DeriveHDRfromLDR}.RemappedRange
    Alpha=Multiply(A=@n68, B=@n69)
   )) {materialFunction=&MF_FX_EmissiveAdjustments}
  n70 = CollectionParameter() {collection=&MPC_Fritter_EnergyWave, parameterName=RandomOpacityFlickSpeed}
  n71 = DynamicParameter() {paramNames [BeamGrowth, WPO_Intensity, SplineWidth, TimeOffset]}
  n72 = LinearInterpolate(
   A=@n66
   B=Multiply(
    A=MaterialFunctionCall!(FlickerInitialSpeed=@n70, FlickerRandom=@n71.TimeOffset) {materialFunction=/Game/Effects/Fort_Effects/Materials/_Functions/MF_Flicker}.Result
    B=@n66
   )
   Alpha=ScalarParameter() {parameterName=UseRandomOpacity}
  ) {constA=1}
  n73 = Constant() {r=1}
  n74 = Saturate(None=Multiply(A=@n38) {constB=4})
  n75 = ScalarParameter() {desc=Switch, parameterName=UseBeamGrowth}
  Multiply(
   A=Multiply(A=Multiply(A=@n60.A, B=@n72), B=LinearInterpolate(A=@n73, B=@n74, Alpha=@n75))
   B=ScalarParameter() {defaultValue=1, parameterName=OpacityOverall}
  )
  Multiply(A=@n50)
  OneMinus(None=@n41.R)
  Sine(None=Multiply(A=@n55.VGradient) {constB=0.5})
 ]
}
